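# Ambient loop sound cue for the ClayTile PristineMask A prop, routed to the loot ambient sound class and concurrency set.
# A mixer layers a looping ambience bed, faded in by an enveloper, under louder random whispers.
# The whispers come from five equally weighted one-shots, repeated by a Looping node with a random delay of a few seconds.
# It references a boss totem attenuation asset and carries override values for HRTF spatialization, listener focus and occlusion.
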SoundCue /Hermes/Props/ClayTile/ClayTile_PristineMask_A/Audio/SC_ClayTile_PristineMask_A_Ambient_Loop
root=Mixer_0
nodes [
 {id=Delay_0, children [Random_0], properties {delayMax=3, delayMin=2}}
 {
  id=Enveloper_0
  children [WavePlayer_3]
  properties {pitchCurve {editorCurveData {keys [{time=0}]}}, volumeCurve {editorCurveData {keys [<at 0 0>,<at 1 1>]}}}
 }
 {id=Looping_0, children [Delay_0]}
 {id=Mixer_0, children [Enveloper_0,Looping_0], properties {inputVolume [0.75,1.25]}}
 {id=Random_0, children [WavePlayer_4,WavePlayer_5,WavePlayer_6,WavePlayer_7,WavePlayer_8], properties {weights [1,1,1,1,1]}}
 {
  id=WavePlayer_3
  properties {bLooping=true, soundWaveAssetPtr=&/ClayTile_PristineMask_A_amb_01}
 }
 {id=WavePlayer_4, properties {soundWaveAssetPtr=&/ClayTile_PristineMask_A_whispers_01}}
 {id=WavePlayer_5, properties {soundWaveAssetPtr=&/ClayTile_PristineMask_A_whispers_02}}
 {id=WavePlayer_6, properties {soundWaveAssetPtr=&/ClayTile_PristineMask_A_whispers_03}}
 {id=WavePlayer_7, properties {soundWaveAssetPtr=&/ClayTile_PristineMask_A_whispers_04}}
 {id=WavePlayer_8, properties {soundWaveAssetPtr=&/ClayTile_PristineMask_A_whispers_05}}
]
properties {
 attenuationSettings=&/Att_ClayTile_PristineMask_Interact_Boss_Totem_Stereo_Att
 maxDistance=2400
 priority=50
 soundClassObject=/Game/Sounds/FortSoundClassesAndModes/Fort_Loot_Ambient_Loop
 volumeMultiplier=1
 attenuationOverrides {
  bEnableFocusInterpolation=true
  bEnableListenerFocus=true
  distanceAlgorithm=Logarithmic
  falloffDistance=1100
  focusAttackInterpSpeed=20
  focusAzimuth=20
  focusReleaseInterpSpeed=10
  focusVolumeAttenuation=3.5
  lPFFrequencyAtMax=10000
  lPFRadiusMax=1000
  lPFRadiusMin=500
  nonFocusAzimuth=40
  nonFocusVolumeAttenuation=0.8
  nonSpatializedRadiusStart=200
  occlusionInterpolationTime=0.3
  occlusionLowPassFilterFrequency=3000
  occlusionVolumeAttenuation=0.75
  spatializationAlgorithm=SPATIALIZATION_HRTF
  attenuationShapeExtents {x=0}
  customAttenuationCurve {editorCurveData {keys [<at 0 1 RCIM_Cubic>, {arriveTangent=-1, interpMode=RCIM_Cubic, leaveTangent=-1, time=0.5, value=0.1}, <at 1 0 RCIM_Cubic>]}}
 }
 concurrencyOverrides {resolutionRule=StopQuietest}
 concurrencySet [/Game/Sounds/Concurrency_Rules/Loot_Ambient_Conc]
}
