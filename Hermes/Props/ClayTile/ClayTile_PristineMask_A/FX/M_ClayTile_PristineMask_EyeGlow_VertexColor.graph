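# Translucent, unlit, two-sided material for the eye glow effect on the ClayTile pristine mask, used on Niagara mesh particles.
# Inverted red vertex color, raised to a dynamic-parameter exponent, forms a glow mask that multiplies both particle color and alpha.
# The masked color passes through DeriveHDRfromLDR, serving as both saturated LDR input and HDR tint, then into MF_FX_EmissiveAdjustments.
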Material /Hermes/Props/ClayTile/ClayTile_PristineMask_A/FX/M_ClayTile_PristineMask_EyeGlow_VertexColor
graph {
 properties {
  bAutomaticallySetUsageInEditor=false
  bUsedWithNiagaraMeshParticles=true
  blendMode=BLEND_Translucent
  previewMesh=&/SM_ClayTile_PristineMask_FX
  shadingModel=MSM_Unlit
  twoSided=true
  usedShadingModels=Unlit
 }
 nodes [
  n1 = ParticleColor()
  n2 = Power(Base=OneMinus(None=VertexColor().R), Exp=DynamicParameter() {paramNames [VertexColor,Param2,Param3,Param4]}.VertexColor)
  n3 = Multiply(A=@n1.RGB, B=@n2)
  MaterialFunctionCall(Emissive=MaterialFunctionCall(LDR Input=Saturate(None=@n3), HDR Tint=@n3) {materialFunction=&DeriveHDRfromLDR}.RemappedRange) {materialFunction=&MF_FX_EmissiveAdjustments}
  Multiply(A=@n1.A, B=@n2)
 ]
}
