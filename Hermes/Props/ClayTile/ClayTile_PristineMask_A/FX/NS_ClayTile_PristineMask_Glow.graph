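# Niagara glow effect for the pristine clay tile mask prop, placing a mirrored pair of magenta glows on it.
# One emitter bursts two particles with looping lifetimes, partitioned to mirror mesh scale, rotation and glow positions.
# Each drives mesh, sprite and light renderers, with waveforms animating a material parameter and flickering sprite alpha.
# A Fortnite time-of-day parameter collection value feeds curves that boost light alpha at the cycle's ends.
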NiagaraSystem /Hermes/Props/ClayTile/ClayTile_PristineMask_A/FX/NS_ClayTile_PristineMask_Glow
dependencies {
 usedDataInterfaces [&NiagaraDataInterfaceCurve]
 usedDynamicInputs [&MakeFloatFromVector2d]
 usedModules [&SystemState]
}
emitters {Glow_Mesh001 {
  emitterUpdateScript [
   &EmitterState(Life Cycle Mode=::NewEnumerator1, Loop Behavior=::NewEnumerator1, Loop Duration Mode=::NewEnumerator1, UseLoopDelay=false)
   &SpawnBurst_Instantaneous(Spawn Count=2, Use Spawn Probability=false, UseLoopCountLimit=false)
  ]
  particleSpawnScript [
   &PartitionParticles()
   &InitializeParticle(
    Color=<rgba 2 0.0532 1.4983 1>
    Lifetime=0.5
    Mesh Scale=&Lerp_Vector(A=<1 1 1>, B=<-1 1 1>, Alpha=&MakeFloatFromInt(Int=$Particles.PartitionParticles.Partition))
    Mesh Scale Mode=::NewEnumerator4
    Position Mode=::NewEnumerator0
    Sprite Rotation Angle=&Multiply_Float(
     A=-0.07
     B=&Lerp_Float(A=1, B=-1, Alpha=&MakeFloatFromInt(Int=$Particles.PartitionParticles.Partition))
    )
    Sprite Rotation Mode=::NewEnumerator1
    Write Material Random=true
   )
  ]
  particleUpdateScript [
   &ParticleState(
    Kill Particles When Lifetime Has Elapsed=false
    Let Infinitely Lived Particles Die When Emitter Deactivates=true
    Loop Particles Lifetime=true
   )
   &DynamicMaterialParameters(
    Index 0 Param 1=&Waveform(
     "Waveform [1]"=::NewEnumerator0
     "[1] Frequency"=&RandomRangeFloat(Minimum=2, Maximum=3)
     "[1] Amplitude Min/Max"=&RandomRangeVector2D(Minimum=<1.2 2>, Maximum=<1.1 3>)
    )
    Module.Index0Float0Write=true
    Module.Index0Float1Write=true
    Module.Index0Float2Write=true
    Module.Index0Float3Write=true
   ) {module=DynamicMaterialParameters001}
   &ScaleColor(Scale Alpha=0.2, ScaleA=true, ScaleRGB=false)
   /SetVariables(Particles.Jitter:NiagaraFloat=&Waveform("Waveform [1]"=::NewEnumerator12, "[1] Phase Scale"=2, "[1] Frequency"=12, "[1] Amplitude Min/Max"=<0.85 1.3>)) {isSetParameters=true}
   /SetVariables(
    Particles.SpriteGlowCameraOffset=9
    Particles.SpriteGlowColor:LinearColor=&MultiplyLinearColor(
     Linear Color=<rgba 4 0.5411 3.6177 1>
     Scale Factor=&MakeVector4(
      X=1
      Y=1
      Z=1
      W=&Multiply_Float(
       A=&FloatFromCurve(
        FloatCurve=DataInterface() {propertyValues {Curve {keys [
            {tangentMode=RCTM_User, time=0, value=1}
            {arriveTangent=-0.0405, interpMode=RCIM_Cubic, leaveTangent=-0.0405, tangentMode=RCTM_User, time=0.7009, value=1}
            {arriveTangent=-8.0089, interpMode=RCIM_Cubic, leaveTangent=-8.0089, tangentMode=RCTM_User, time=1, value=1}
           ]}}}
        CurveIndex=$Emitter.Age
        Scale Curve=&Multiply_Float(A=$Particles.Jitter, B=0.4)
       )
       B=&FloatFromCurve(
        FloatCurve=DataInterface() {propertyValues {Curve {keys [
            {interpMode=RCIM_Cubic, tangentMode=RCTM_User, time=0, value=1}
            {interpMode=RCIM_Cubic, tangentMode=RCTM_User, time=0.5, value=1}
            {tangentMode=RCTM_User, time=1, value=1}
           ]}}}
        CurveIndex=Linked() {linkedVariable {name=NPC.FortniteNPC.TimeOfDay_0-1, type=NiagaraFloat}}
       )
      )
     )
    )
    Particles.SpriteGlowPosition:NiagaraPosition=&Lerp_Vector(A=<13 28 3>, B=<-13 28 3>, Alpha=&MakeFloatFromInt(Int=$Particles.PartitionParticles.Partition))
    Particles.SpriteGlowScale:Vector2f=&Multiply_Vector2D_ByFloat(Float=20, Vector2D=<1.6 1>)
   ) {isSetParameters=true}
   /SetVariables(
    Particles.LightColor=&MultiplyLinearColor(
     Linear Color=<rgba 3 0 3 1>
     Scale Factor=&MakeVector4(
      X=1
      Y=1
      Z=1
      W=&FloatFromCurve(
       FloatCurve=DataInterface() {propertyValues {Curve {keys [
           {interpMode=RCIM_Cubic, tangentMode=RCTM_User, time=0, value=1.5}
           {interpMode=RCIM_Cubic, tangentMode=RCTM_User, time=0.5, value=1}
           {tangentMode=RCTM_User, time=1, value=1.5}
          ]}}}
       CurveIndex=Linked() {linkedVariable {name=NPC.FortniteNPC.TimeOfDay_0-1, type=NiagaraFloat}}
      )
     )
    )
    Particles.LightPosition:NiagaraPosition=&Lerp_Vector(A=<13 28 3>, B=<-13 28 3>, Alpha=&MakeFloatFromInt(Int=$Particles.PartitionParticles.Partition))
   ) {isSetParameters=true}
  ]
  rendererClasses [
   &NiagaraMeshRendererProperties
   &NiagaraSpriteRendererProperties
   &NiagaraLightRendererProperties
  ]
  renderers [
   {rendererClass=&NiagaraMeshRendererProperties, rendererIndex=0}
   {rendererClass=&NiagaraSpriteRendererProperties, rendererIndex=1}
   {rendererClass=&NiagaraLightRendererProperties, rendererIndex=2}
  ]
 }}
summary {
 systemName=NS_ClayTile_PristineMask_Glow
 emitters [{
   emitterName=Glow_Mesh001
   rendererClasses [
    &NiagaraMeshRendererProperties
    &NiagaraSpriteRendererProperties
    &NiagaraLightRendererProperties
   ]
  }]
 userVariables []
}
